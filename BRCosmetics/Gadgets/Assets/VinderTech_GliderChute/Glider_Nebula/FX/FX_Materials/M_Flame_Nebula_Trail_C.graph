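# Translucent unlit flame trail material for the Nebula glider's FX, usable on Niagara ribbons and mesh particles.
# Two multiplied, panning Voronoi samples distort a tapered shape and scrolling trail mask, colored through curve-atlas and gradient lookups with optional flat-color and galaxy blends.
# Dynamic parameters drive UV offset, dive emissive boost, camera offset and particle-color blend.
# Layered sines wobble vertices, while depth, camera, view-angle and optional fresnel fades soften opacity.
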Material /BRCosmetics/Gadgets/Assets/VinderTech_GliderChute/Glider_Nebula/FX/FX_Materials/M_Flame_Nebula_Trail_C
graph {
 properties {
  bCastRayTracedShadows=false
  bUsedWithNiagaraMeshParticles=true
  bUsedWithNiagaraRibbons=true
  blendMode=BLEND_Translucent
  pixelDepthOffsetMode=PDOM_Legacy
  shadingModel=MSM_Unlit
  usedShadingModels=Unlit
 }
 nodes [
  n1 = MaterialFunctionCall(Offset Amount=DynamicParameter() {defaultValue=<rgba 0.8 1 0 0>, paramNames [UV1 U Offset, Dive, Camera Offset, BlendParticleCOlor]}.Camera Offset) {materialFunction=/Engine/Functions/Engine_MaterialFunctions02/WorldPositionOffset/CameraOffset}
  n2 = Add(
   A=NamedRerouteUsage()
   B=AppendVector(
    A=DynamicParameter() {defaultValue=<rgba 0.8 1 0 0>, paramNames [UV1 U Offset, Dive, Camera Offset, BlendParticleCOlor]}.UV1 U Offset
    B=Constant()
   )
  )
  n3 = Panner(Coordinate=@n2) {speedX=-0.25}
  n4 = ComponentMask(None=Panner(Coordinate=@n2) {speedX=-0.175}) {r=true}
  n5 = Add(A=Sine(None=ComponentMask(None=@n3) {r=true}) {period=1.4}, B=Multiply(A=Sine(None=@n4) {period=0.5}) {constB=0.3})
  n6 = Panner() {speedX=-0.25}
  n7 = Multiply(A=Sine(None=ComponentMask(None=@n6) {r=true}) {period=3}) {constB=0.7}
  n8 = VectorParameter() {group=WPO, parameterName=WPO Direction, defaultValue=<rgba 1 1 0 1>}
  n9 = NamedRerouteUsage()
  n10 = Multiply(
   A=Multiply(A=Add(A=@n5, B=@n7), B=Transform(None=@n8.RGB) {transformSourceType=TRANSFORMSOURCE_Instance})
   B=Power(Base=ComponentMask(None=@n9) {r=true})
  )
  n11 = VertexNormalWS()
  n12 = OneMinus(None=NamedRerouteUsage())
  Add(
   A=Multiply(A=Add(A=@n1.Clamped Camera Offset, B=@n10) {constB=0}, B=ScalarParameter() {defaultValue=1, parameterName=WPO Weight, sliderMax=1})
   B=Multiply(A=Multiply(A=@n11, B=@n12), B=ScalarParameter() {parameterName=WPO Noisy, sliderMax=25})
  )
  Constant2Vector() {g=0.15, r=0.75}
  Constant() {r=1}
  n13 = NamedRerouteUsage()
  n14 = ScalarParameter() {parameterName=AddGradient}
  n15 = Multiply(
   A=TextureSampleParameter2D(UVs=Add(A=@n13, B=@n14) {constB=0.1}) {parameterName=GradientTexture, texture=/Game/Effects/Fort_Effects/Textures/T_Gradient_Nebula}.RGB
   B=ScalarParameter() {defaultValue=0.2, parameterName=GradientScale, sliderMax=1}
  ) {constB=0.5}
  n16 = NamedRerouteUsage()
  n17 = NamedRerouteUsage()
  n18 = ScalarParameter() {defaultValue=2, parameterName=Vertical Gradient Tiling, sliderMax=3}
  n19 = DynamicParameter() {defaultValue=<rgba 0.8 1 0 0>, paramNames [UV1 U Offset, Dive, Camera Offset, BlendParticleCOlor]}
  n20 = Constant()
  n21 = Panner(
   Coordinate=NamedRerouteUsage()
   Speed=AppendVector(
    A=ScalarParameter() {defaultValue=-3, group=UV Distortion, parameterName=UV Speed X 2, sliderMax=5, sliderMin=-5}
    B=ScalarParameter() {group=UV Distortion, parameterName=UV Speed Y 2, sliderMax=5, sliderMin=-5}
   )
  ) {speedX=-0.5}
  n22 = AppendVector(
   A=ScalarParameter() {defaultValue=0.25, group=UV Distortion, parameterName=UV Scale X 2, sliderMax=5}
   B=ScalarParameter() {defaultValue=0.5, group=UV Distortion, parameterName=UV Scale Y 2, sliderMax=5}
  )
  n23 = TextureSampleParameter2D(UVs=Add(A=AppendVector(A=@n19.UV1 U Offset, B=@n20), B=Multiply(A=@n21, B=@n22))) {
   group=UV Distortion
   parameterName=Distortion Texture
   texture=/BRCosmetics/Athena/Effects/Particles/Skydiving/NS_Contrail_PinkSpike/Textures/T_PinkSpike_Voronoi2D_Height
  }
  n24 = DynamicParameter() {defaultValue=<rgba 0.8 1 0 0>, paramNames [UV1 U Offset, Dive, Camera Offset, BlendParticleCOlor]}
  n25 = Constant()
  n26 = Panner(
   Coordinate=NamedRerouteUsage()
   Speed=AppendVector(
    A=ScalarParameter() {defaultValue=-2, group=UV Distortion, parameterName=UV Speed X, sliderMax=5, sliderMin=-5}
    B=ScalarParameter() {group=UV Distortion, parameterName=UV Speed Y, sliderMax=5, sliderMin=-5}
   )
  ) {speedX=-0.5}
  n27 = AppendVector(
   A=ScalarParameter() {defaultValue=0.4, group=UV Distortion, parameterName=UV Scale X, sliderMax=5}
   B=ScalarParameter() {defaultValue=0.8, group=UV Distortion, parameterName=UV Scale Y, sliderMax=5}
  )
  n28 = TextureSampleParameter2D(UVs=Add(A=AppendVector(A=@n24.UV1 U Offset, B=@n25), B=Multiply(A=@n26, B=@n27))) {
   group=UV Distortion
   parameterName=Distortion Texture
   texture=&T_PinkSpike_Voronoi2D_Height
  }
  n29 = MaterialFunctionCall(X=Multiply(A=@n23.R, B=@n28.R), Y=Multiply(A=@n23.G, B=@n28.G)) {materialFunction=&MakeFloat2}
  n30 = Subtract(A=OneMinus(None=Power(Base=@n29.Result))) {constB=0.25}
  n31 = ComponentMask(None=Add(A=Multiply(A=@n17, B=@n18), B=@n30)) {r=true}
  n32 = CurveAtlasRowParameter(CurveTime=Saturate(None=OneMinus(None=@n31))) {atlas=/Game/Athena/LinearColorCurves/CA_Nebula_Contrail, curve=/Game/Athena/LinearColorCurves/CC_Nebula_Contrail, parameterName=Curve_Color}
  n33 = ScalarParameter() {defaultValue=0.5, parameterName=Vertical Gradient Bloom, sliderMax=1}
  n34 = ComponentMask(None=@n17) {r=true}
  n35 = NamedRerouteUsage()
  n36 = Add(A=@n35, B=Multiply(A=@n30) {constB=0.1})
  n37 = ComponentMask(None=@n36) {r=true}
  n38 = OneMinus(None=ComponentMask(None=@n35) {r=true})
  n39 = ScalarParameter() {defaultValue=0.8, parameterName=TaperPower, sliderMax=3}
  n40 = Multiply(
   A=AppendVector(A=OneMinus(None=@n37), B=ComponentMask(None=@n36) {g=true})
   B=AppendVector(A=Power(Base=@n38, Exp=@n39) {constExponent=1}, B=Constant() {r=1})
  )
  n41 = Constant2Vector() {g=0.5, r=0.5}
  n42 = Multiply(A=SphereMask(A=@n40, B=@n41) {attenuationRadius=0.5, hardnessPercent=0}) {constB=2}
  n43 = ScalarParameter() {defaultValue=-0.2, group=UV Distortion, parameterName=Distortion Strength, sliderMax=1, sliderMin=-1}
  n44 = Panner(
   Coordinate=Add(A=NamedRerouteUsage(), B=Multiply(A=@n30, B=@n43) {constB=0.1})
   Speed=AppendVector(
    A=ScalarParameter() {defaultValue=-2.5, group=Flame, parameterName=Speed X, sliderMax=5, sliderMin=-5}
    B=ScalarParameter() {group=Flame, parameterName=Speed Y, sliderMax=5, sliderMin=-5}
   )
  ) {speedX=1}
  n45 = AppendVector(
   A=ScalarParameter() {defaultValue=0.2, group=Flame, parameterName=Scale X, sliderMax=5}
   B=ScalarParameter() {defaultValue=0.3, group=Flame, parameterName=Scale Y, sliderMax=5}
  )
  n46 = DynamicParameter() {defaultValue=<rgba 0.8 1 0 0>, paramNames [UV1 U Offset, Dive, Camera Offset, BlendParticleCOlor]}
  n47 = ScalarParameter() {defaultValue=0.2, parameterName=Shift Y, sliderMax=1}
  n48 = TextureSampleParameter2D(UVs=Add(A=Multiply(A=@n44, B=@n45), B=AppendVector(A=@n46.UV1 U Offset, B=@n47))) {
   parameterName=Mask Texture
   samplerType=SAMPLERTYPE_LinearGrayscale
   texture=/BRCosmetics/Athena/Effects/Particles/Skydiving/NS_Contrail_Nebula/T_Nebula_Trail
  }
  n49 = Multiply(A=@n42, B=@n48.R)
  n50 = NamedRerouteUsage()
  n51 = OneMinus(None=ComponentMask(None=@n50) {r=true})
  n52 = ComponentMask(None=@n50) {g=true}
  n53 = Multiply(A=@n52, B=OneMinus(None=@n52))
  n54 = Multiply(A=Multiply(A=Saturate(None=@n51)) {constB=0.5}, B=Saturate(None=Multiply(A=@n53) {constB=3}))
  n55 = ScalarParameter() {defaultValue=50, parameterName=Camera Fade Length, sliderMax=200}
  n56 = TextureCoordinate()
  n57 = Subtract(A=Power(Base=ComponentMask(None=@n56) {r=true}) {constExponent=1}) {constB=0.1}
  n58 = Multiply(
   A=Power(
    Base=LinearInterpolate(A=@n49, B=@n42, Alpha=@n54)
    Exp=ScalarParameter() {defaultValue=1.5, group=Opacity, parameterName=OpacityExponent, sliderMax=4}
   )
   B=Multiply(
    A=MaterialFunctionCall(Fade Length=@n55) {materialFunction=&CameraDepthFade}.Result
    B=LinearInterpolate(B=@n57) {constA=1}
   ) {constA=1}
  )
  n59 = SmoothStep(Value=Multiply(A=OneMinus(None=@n34), B=@n58)) {constMax=1.5, constMin=0.1}
  n60 = CurveAtlasRowParameter(CurveTime=@n59) {atlas=&CA_Nebula_Contrail, curve=&CC_Nebula_Contrail, parameterName=Curve_Color}
  n61 = ScalarParameter() {defaultValue=0.75, parameterName=Emissive, sliderMax=2}
  n62 = ScalarParameter() {defaultValue=1.2, parameterName=Emissive_Dive, sliderMax=3}
  n63 = DynamicParameter() {defaultValue=<rgba 0.8 1 0 0>, paramNames [UV1 U Offset, Dive, Camera Offset, BlendParticleCOlor]}
  n64 = Add(A=Multiply(A=@n59, B=LinearInterpolate(A=@n61, B=@n62, Alpha=@n63.Dive) {constA=1})) {constB=0.5}
  n65 = Add(
   A=LinearInterpolate(A=@n15, B=Multiply(A=@n15, B=@n16), Alpha=ScalarParameter() {defaultValue=1, parameterName=EdgeDarken, sliderMax=1})
   B=Add(A=Multiply(A=@n32, B=@n33), B=Multiply(A=@n60, B=@n64))
  )
  n66 = ScalarParameter() {group=FlatColor, parameterName=FlatColorMix, sliderMax=1, sortPriority=1}
  n67 = VectorParameter() {group=FlatColor, parameterName=FlatColor, sortPriority=3, defaultValue=<rgba 0.2999 0 0.8177 1>}
  n68 = VectorParameter() {group=FlatColor, parameterName=FlatColor_Root, sortPriority=4, defaultValue=<rgba 0.2999 0 0.8177 1>}
  n69 = OneMinus(None=TextureCoordinate())
  n70 = ScalarParameter() {defaultValue=1, group=FlatColor, parameterName=FlatColor Root Exp, sliderMax=1}
  n71 = LinearInterpolate(
   A=@n67.RGB
   B=LinearInterpolate(A=@n67.RGB, B=@n68.RGB, Alpha=Power(Base=@n69, Exp=@n70))
   Alpha=ScalarParameter() {group=FlatColor, parameterName=FlatColor RootWeight, sliderMax=1, sortPriority=5}
  )
  n72 = VectorParameter() {parameterName=TOD color, defaultValue=<rgba 1 1 1 0.5>}
  n73 = LinearInterpolate(
   A=Multiply(
    A=LinearInterpolate(A=@n65, B=@n71, Alpha=@n66)
    B=MaterialFunctionCall(Value_Morning=@n72.R, Value_Day=@n72.R, Value_Evening=@n72.R, Value_Night=@n72.R) {materialFunction=&Scalar_Time_of_Day}.Result
   )
   B=ParticleColor().RGB
   Alpha=DynamicParameter() {defaultValue=<rgba 0.8 1 0 0>, paramNames [UV1 U Offset, Dive, Camera Offset, BlendParticleCOlor]}.BlendParticleCOlor
  )
  n74 = BreakMaterialAttributes!(Attr=MaterialFunctionCall!(MA=SetMaterialAttributes!(MaterialAttributes), SkinFX=Constant4Vector() {constant=<rgba 1 1 1 0>}.RGBA) {materialFunction=/Game/Packages/Fortress_SharedMaterials/Base_Material_Functions/MFC_GalaxyScreenSpace}.Result)
  n75 = ScalarParameter() {parameterName=GalaxyBlend}
  MaterialFunctionCall(Emissive=Multiply(
    A=LinearInterpolate(A=@n73, B=@n74.EmissiveColor, Alpha=@n75)
    B=ScalarParameter() {defaultValue=0.8, parameterName=FinalEmissiveScale, sliderMax=1}
   )) {materialFunction=&MF_FX_EmissiveAdjustments}
  Multiply(A=@n32, B=@n32)
  n76 = SmoothStep(Max=ScalarParameter() {defaultValue=0.55, group=Opacity, parameterName=OpacityFit, sliderMax=1}, Value=@n58) {constMax=0.55}
  n77 = DepthFade(Opacity=@n76, FadeDistance=ScalarParameter() {defaultValue=20, parameterName=Depth Fade Distance, sliderMax=100}) {fadeDistanceDefault=25}
  n78 = Multiply(
   A=Multiply(
    A=ShadingPathSwitch(Default=@n77, Deferred=@n77, Forward=@n76, Mobile=@n76)
    B=ScalarParameter() {defaultValue=0.8, group=Opacity, parameterName=Opacity, sliderMax=1}
   )
   B=ParticleColor().A
  )
  n79 = OneMinus(None=Fresnel(ExponentIn=ScalarParameter() {defaultValue=12, parameterName=Fresnel Opacity Exp, sliderMax=15}) {exponent=12})
  Multiply(
   A=LinearInterpolate(A=@n78, B=Multiply(A=@n78, B=@n79), Alpha=ScalarParameter() {parameterName=Fresnel Opacity Mix, sliderMax=1})
   B=MaterialFunctionCall("FadeSensitivity (optional)"=ScalarParameter() {defaultValue=1.5, parameterName=FadeAngleIntensity}) {materialFunction=&FadeBasedOnViewAngle}.Result
  )
  NamedRerouteDeclaration(None=@n58) {name=ShapeMask, nodeColor=<rgba 1 0.8203 0 1>}
  n80 = TextureCoordinate() {coordinateIndex=1}
  n81 = MaterialFunctionCall(XYZ=@n80) {materialFunction=&Swizzle}
  n82 = ScalarParameter() {parameterName=SwizzleUV1, sliderMax=1}
  n83 = MaterialFunctionCall(Float2=NamedRerouteDeclaration(None=LinearInterpolate(A=@n80, B=@n81.YX, Alpha=@n82)) {name=UV1, nodeColor=<rgba 1 0.1108 0 1>}) {materialFunction=&BreakOutFloat2Components}
  n84 = Multiply(A=@n83.R, B=ScalarParameter() {defaultValue=1, group=TexCoords, parameterName=TexCoord1_ScaleX, sliderMax=1})
  n85 = Multiply(A=@n83.G, B=ScalarParameter() {defaultValue=1, group=TexCoords, parameterName=TexCoord1_ScaleY, sliderMax=1})
  n86 = Multiply(A=VertexColor().R, B=ScalarParameter() {group=Flame, parameterName=Vertex Color Shift, sliderMax=1})
  n87 = Constant()
  NamedRerouteDeclaration(None=Add(
    A=MaterialFunctionCall(X=@n84, Y=@n85) {materialFunction=&MakeFloat2}.Result
    B=AppendVector(A=@n86, B=@n87)
   )) {name=TexCoord_1_scaled, nodeColor=<rgba 1 0 0.5751 1>}
  NamedRerouteDeclaration(None=SphereMask(A=@n40, B=Constant2Vector() {g=0.5, r=0.65}, Radius=@n41) {attenuationRadius=0.5, hardnessPercent=0}) {name=temp, nodeColor=<rgba 0.0781 1 0 1>}
  n88 = TextureCoordinate()
  n89 = AppendVector(
   A=ScalarParameter() {defaultValue=0.5, group=TexCoords, parameterName=Texcoord0 ScalePivotY, sliderMax=1}
   B=ScalarParameter() {defaultValue=0.5, group=TexCoords, parameterName=Texcoord0 ScalePivotX, sliderMax=1}
  )
  n90 = Subtract(
   A=LinearInterpolate(
    A=@n88
    B=MaterialFunctionCall(XYZ=@n88) {materialFunction=&Swizzle}.YX
    Alpha=ScalarParameter() {parameterName=SwizzleUV0, sliderMax=1}
   )
   B=@n89
  )
  n91 = AppendVector(
   A=ScalarParameter() {defaultValue=1, group=TexCoords, parameterName=Texcoord0 ScaleY, sliderMax=2}
   B=ScalarParameter() {defaultValue=1, group=TexCoords, parameterName=Texcoord0 ScaleX, sliderMax=2}
  )
  NamedRerouteDeclaration(None=Add(A=Multiply(A=@n90, B=@n91), B=@n89)) {name=UV0, nodeColor=<rgba 1 0.0171 0 1>}
  NamedRerouteUsage()
  ScalarParameter() {parameterName=UseParticleCOlor}
  TextureCoordinate() {uTiling=4}
  TextureCoordinate() {coordinateIndex=1}
  TextureCoordinate() {coordinateIndex=1}
  TextureCoordinate()
  TextureCoordinate()
  TextureCoordinate()
  TextureCoordinate()
  TextureCoordinate()
 ]
}
